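# Sound cue for the hit notification when a critical hit is received, capped by a two-maximum, stop-oldest concurrency set.
# A quality switch selects the full random variation set, a reduced set, or one "opt" wave for both remaining levels.
# The layered branches mix volume-enveloped rifle bullet impacts with a delayed pair of enemy impact and quieter, double-pitched metal cannon impact.
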SoundCue /FigmentCore/Sounds/HitNotifications/HitMarker_Critical_Receive_Cue
root=QualityLevel_0
nodes [
 {id=Delay_0, children [Mixer_1], properties {delayMax=0.15, delayMin=0.15}}
 {id=Delay_1, children [Mixer_3], properties {delayMax=0.15, delayMin=0.15}}
 {
  id=Enveloper_0
  children [Random_0]
  properties {
   pitchCurve {editorCurveData {keys [{time=0}]}}
   volumeCurve {editorCurveData {keys [
      <at 0 1>
      {arriveTangent=32.5924, interpMode=RCIM_Cubic, leaveTangent=32.5924, tangentMode=RCTM_User, time=0.3, value=0}
      {arriveTangent=2.8166, interpMode=RCIM_Cubic, leaveTangent=2.8166, tangentMode=RCTM_User, time=0.4, value=1}
     ]}}
  }
 }
 {
  id=Enveloper_1
  children [Random_4]
  properties {
   pitchCurve {editorCurveData {keys [{time=0}]}}
   volumeCurve {editorCurveData {keys [
      <at 0 1>
      {arriveTangent=32.5924, interpMode=RCIM_Cubic, leaveTangent=32.5924, tangentMode=RCTM_User, time=0.3, value=0}
      {arriveTangent=2.8166, interpMode=RCIM_Cubic, leaveTangent=2.8166, tangentMode=RCTM_User, time=0.4, value=1}
     ]}}
  }
 }
 {id=Mixer_0, children [Enveloper_0,Modulator_0], properties {inputVolume [1,1]}}
 {id=Mixer_1, children [Random_1,Modulator_1], properties {inputVolume [1,0.7]}}
 {id=Mixer_2, children [Enveloper_1,Modulator_3], properties {inputVolume [1,1]}}
 {id=Mixer_3, children [Random_5,Modulator_4], properties {inputVolume [1,0.7]}}
 {id=Modulator_0, children [Delay_0], properties {pitchMax=1, pitchMin=1}}
 {id=Modulator_1, children [Random_2], properties {pitchMax=2, pitchMin=2}}
 {id=Modulator_3, children [Delay_1], properties {pitchMax=1, pitchMin=1}}
 {id=Modulator_4, children [Random_6], properties {pitchMax=2, pitchMin=2}}
 {id=QualityLevel_0, children [Mixer_2,Mixer_0,WavePlayer_0,WavePlayer_0]}
 {id=Random_0, children [WavePlayer_19, WavePlayer_20], properties {bRandomizeWithoutReplacement=false, weights [1,1]}}
 {id=Random_1, children [WavePlayer_18, WavePlayer_17], properties {weights [1,1]}}
 {id=Random_2, children [WavePlayer_16, WavePlayer_15], properties {bRandomizeWithoutReplacement=false, weights [1,1]}}
 {id=Random_4, children [WavePlayer_2,WavePlayer_3,WavePlayer_4,WavePlayer_5], properties {weights [1,1,1,1]}}
 {id=Random_5, children [WavePlayer_6,WavePlayer_7,WavePlayer_8,WavePlayer_9], properties {weights [1,1,1,1]}}
 {id=Random_6, children [WavePlayer_10, WavePlayer_11, WavePlayer_12, WavePlayer_13, WavePlayer_14], properties {weights [1,1,1,1,1]}}
 {id=WavePlayer_0, properties {soundWaveAssetPtr=&/HitMarker_Critical_Receive_opt}}
 {id=WavePlayer_10, properties {soundWaveAssetPtr=&/Metal_Cannon_Impact01}}
 {id=WavePlayer_11, properties {soundWaveAssetPtr=&/Metal_Cannon_Impact02}}
 {id=WavePlayer_12, properties {soundWaveAssetPtr=&/Metal_Cannon_Impact03}}
 {id=WavePlayer_13, properties {soundWaveAssetPtr=&/Metal_Cannon_Impact04}}
 {id=WavePlayer_14, properties {soundWaveAssetPtr=&/Metal_Cannon_Impact05}}
 {id=WavePlayer_15, properties {soundWaveAssetPtr=&Metal_Cannon_Impact02}}
 {id=WavePlayer_16, properties {soundWaveAssetPtr=&Metal_Cannon_Impact01}}
 {id=WavePlayer_17, properties {soundWaveAssetPtr=&/player_impact_enemy_stereo_02}}
 {id=WavePlayer_18, properties {soundWaveAssetPtr=&/player_impact_enemy_stereo_01}}
 {id=WavePlayer_19, properties {soundWaveAssetPtr=&/player_bulletimpact_rifle_01}}
 {id=WavePlayer_2, properties {soundWaveAssetPtr=&player_bulletimpact_rifle_01}}
 {id=WavePlayer_20, properties {soundWaveAssetPtr=&/player_bulletimpact_rifle_02}}
 {id=WavePlayer_3, properties {soundWaveAssetPtr=&player_bulletimpact_rifle_02}}
 {id=WavePlayer_4, properties {soundWaveAssetPtr=&/player_bulletimpact_rifle_03}}
 {id=WavePlayer_5, properties {soundWaveAssetPtr=&/player_bulletimpact_rifle_04}}
 {id=WavePlayer_6, properties {soundWaveAssetPtr=&player_impact_enemy_stereo_01}}
 {id=WavePlayer_7, properties {soundWaveAssetPtr=&player_impact_enemy_stereo_02}}
 {id=WavePlayer_8, properties {soundWaveAssetPtr=&/player_impact_enemy_stereo_03}}
 {id=WavePlayer_9, properties {soundWaveAssetPtr=&/player_impact_enemy_stereo_04}}
]
properties {
 attenuationSettings=/Game/Athena/Sounds/HitNotifications/Fort_HitMarker_Receive_Att
 maxDistance=2097152
 soundClassObject=/Game/Sounds/FortSoundClassesAndModes/Fort_2DHitNotifications_Receive_PSM
 concurrencySet [&HitNotifications_TwoMaximum_StopOldest_VolumeScale]
}
