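# Sound cue for the received hit-marker notification of a death by falling.
# The first two quality levels mix a random rifle bullet impact with fading volume, a delayed random bone crunch at doubled volume, and a later big-hit marker.
# The other two quality levels play a single optimized wave instead.
# Receive-specific attenuation and sound class apply, with concurrency limited to two, stopping the oldest.
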SoundCue /FigmentCore/Sounds/HitNotifications/HitMarker_Death_Fall_Receive_Cue
root=QualityLevel_0
nodes [
 {id=Delay_1, children [WavePlayer_9], properties {delayMax=0.5, delayMin=0.5}}
 {id=Delay_2, children [Random_2], properties {delayMax=0.25, delayMin=0.25}}
 {
  id=Enveloper_0
  children [Random_1]
  properties {pitchCurve {editorCurveData {keys [{time=0}]}}, volumeCurve {editorCurveData {keys [<at 0 1>,<at 0.75 0>]}}}
 }
 {id=Mixer_0, children [Delay_1,Enveloper_0,Delay_2], properties {inputVolume [1,1,2]}}
 {id=QualityLevel_0, children [Mixer_0,Mixer_0,WavePlayer_2,WavePlayer_2]}
 {id=Random_1, children [WavePlayer_3,WavePlayer_4,WavePlayer_5,WavePlayer_6], properties {weights [1,1,1,1]}}
 {id=Random_2, children [WavePlayer_7,WavePlayer_8], properties {weights [1,1]}}
 {id=WavePlayer_2, properties {soundWaveAssetPtr=&/HitMarker_Death_Fall_Receive_opt}}
 {id=WavePlayer_3, properties {soundWaveAssetPtr=&/player_bulletimpact_rifle_01}}
 {id=WavePlayer_4, properties {soundWaveAssetPtr=&/player_bulletimpact_rifle_02}}
 {id=WavePlayer_5, properties {soundWaveAssetPtr=&/player_bulletimpact_rifle_03}}
 {id=WavePlayer_6, properties {soundWaveAssetPtr=&/player_bulletimpact_rifle_04}}
 {id=WavePlayer_7, properties {soundWaveAssetPtr=&/sfx_Falling_BoneCrunch_01}}
 {id=WavePlayer_8, properties {soundWaveAssetPtr=&/sfx_Falling_BoneCrunch_02}}
 {id=WavePlayer_9, properties {soundWaveAssetPtr=&/sfx_hitmarker_bighit}}
]
properties {
 attenuationSettings=/Game/Athena/Sounds/HitNotifications/Fort_HitMarker_Receive_Att
 maxDistance=2097152
 soundClassObject=/Game/Sounds/FortSoundClassesAndModes/Fort_2DHitNotifications_Receive_PSM
 concurrencySet [&HitNotifications_TwoMaximum_StopOldest_VolumeScale]
}
