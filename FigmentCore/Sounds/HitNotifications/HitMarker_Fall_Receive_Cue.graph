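# Sound cue for the hit notification heard when the player receives fall damage.
# A quality-level node picks between two layered mixes and, for the other two levels, a single optimized wave through a modulator.
# Each mix layers a random bone crunch over a quieter, briefly delayed, fading rifle bullet impact, with fewer impact variants in the second.
# It uses a receive-notification sound class and two-voice, stop-oldest concurrency.
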SoundCue /FigmentCore/Sounds/HitNotifications/HitMarker_Fall_Receive_Cue
root=QualityLevel_0
nodes [
 {id=Delay_1, children [Enveloper_0], properties {delayMax=0.05, delayMin=0.05}}
 {id=Delay_2, children [Enveloper_1], properties {delayMax=0.05, delayMin=0.05}}
 {
  id=Enveloper_0
  children [Random_2]
  properties {pitchCurve {editorCurveData {keys [{time=0}]}}, volumeCurve {editorCurveData {keys [<at 0 1>,<at 0.5 0>]}}}
 }
 {
  id=Enveloper_1
  children [Random_1]
  properties {pitchCurve {editorCurveData {keys [{time=0}]}}, volumeCurve {editorCurveData {keys [<at 0 1>,<at 0.5 0>]}}}
 }
 {id=Mixer_0, children [Delay_2,Random_0], properties {inputVolume [0.5,2]}}
 {id=Mixer_1, children [Delay_1,Random_3], properties {inputVolume [0.5,2]}}
 {id=Modulator_1, children [WavePlayer_0]}
 {id=QualityLevel_0, children [Mixer_0,Mixer_1,Modulator_1,Modulator_1]}
 {id=Random_0, children [WavePlayer_14, WavePlayer_15], properties {weights [1,1]}}
 {id=Random_1, children [WavePlayer_5,WavePlayer_6,WavePlayer_7,WavePlayer_8], properties {weights [1,1,1,1]}}
 {id=Random_2, children [WavePlayer_16, WavePlayer_17], properties {weights [1,1]}}
 {id=Random_3, children [WavePlayer_18, WavePlayer_19], properties {weights [1,1]}}
 {id=WavePlayer_0, properties {soundWaveAssetPtr=&/HitMarker_Fall_Receive_opt}}
 {id=WavePlayer_14, properties {soundWaveAssetPtr=&/sfx_Falling_BoneCrunch_05}}
 {id=WavePlayer_15, properties {soundWaveAssetPtr=&/sfx_Falling_BoneCrunch_06}}
 {id=WavePlayer_16, properties {soundWaveAssetPtr=&/player_bulletimpact_rifle_01}}
 {id=WavePlayer_17, properties {soundWaveAssetPtr=&/player_bulletimpact_rifle_02}}
 {id=WavePlayer_18, properties {soundWaveAssetPtr=&sfx_Falling_BoneCrunch_05}}
 {id=WavePlayer_19, properties {soundWaveAssetPtr=&sfx_Falling_BoneCrunch_06}}
 {id=WavePlayer_5, properties {soundWaveAssetPtr=&player_bulletimpact_rifle_01}}
 {id=WavePlayer_6, properties {soundWaveAssetPtr=&player_bulletimpact_rifle_02}}
 {id=WavePlayer_7, properties {soundWaveAssetPtr=&/player_bulletimpact_rifle_03}}
 {id=WavePlayer_8, properties {soundWaveAssetPtr=&/player_bulletimpact_rifle_04}}
]
properties {
 attenuationSettings=/Game/Athena/Sounds/HitNotifications/Fort_HitMarker_Receive_Att
 maxDistance=2097152
 soundClassObject=/Game/Sounds/FortSoundClassesAndModes/Fort_2DHitNotifications_Receive_PSM
 concurrencySet [&HitNotifications_TwoMaximum_StopOldest_VolumeScale]
}
