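# Sound cue for the body-shot hitmarker, giving the player audio confirmation of a hit.
# A Random node picks one of three equally weighted body hit waves, feeding a Modulator and then a short fixed Delay at the root.
# It uses an instigator hit-notification attenuation and 2D sound class, with a concurrency set named for capping at two voices, stopping the oldest and scaling volume.
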SoundCue /FigmentCore/Sounds/HitNotifications/Hitmarker_Body_Cue
root=Delay_0
nodes [
 {id=Delay_0, children [Modulator_0], properties {delayMax=0.1, delayMin=0.1}}
 {id=Modulator_0, children [Random_0]}
 {id=Random_0, children [WavePlayer_0,WavePlayer_1,WavePlayer_2], properties {weights [1,1,1]}}
 {id=WavePlayer_0, properties {soundWaveAssetPtr=&/HitNotification_New_Body_01}}
 {id=WavePlayer_1, properties {soundWaveAssetPtr=&/HitNotification_New_Body_02}}
 {id=WavePlayer_2, properties {soundWaveAssetPtr=&/HitNotification_New_Body_03}}
]
properties {
 attenuationSettings=/Game/Athena/Sounds/HitNotifications/Att_HitMarker_Instigate_MS
 maxDistance=2097152
 soundClassObject=&Fort_2DHitNotifications_Instigate_PSM
 concurrencySet [&HitNotifications_TwoMaximum_StopOldest_VolumeScale]
}
