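# Sound cue for the critical-hit hitmarker notification, routed to an instigator 2D hit-notification sound class with concurrency that stops the oldest.
# A quality switch selects between two layered mixes, the second with fewer random variants, and one optimized wave shared by the two remaining levels.
# Each mix plays a pitched-up random metal cannon impact, then after a short fixed delay a random enemy impact mixed with a louder critical notification layer.
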SoundCue /FigmentCore/Sounds/HitNotifications/Hitmarker_Critical_Cue
root=QualityLevel_0
nodes [
 {id=Delay_1, children [Mixer_3], properties {delayMax=0.15, delayMin=0.15}}
 {id=Delay_6, children [Mixer_1], properties {delayMax=0.15, delayMin=0.15}}
 {id=Mixer_0, children [Modulator_1,Modulator_6], properties {inputVolume [1,1]}}
 {id=Mixer_1, children [Random_7,WavePlayer_9], properties {inputVolume [1,1.25]}}
 {id=Mixer_2, children [Modulator_3,Modulator_2], properties {inputVolume [1,1]}}
 {id=Mixer_3, children [Random_1, WavePlayer_16], properties {inputVolume [1,1.25]}}
 {id=Modulator_1, children [Random_8], properties {pitchMax=2, pitchMin=2}}
 {id=Modulator_2, children [Delay_1], properties {pitchMax=1, pitchMin=1}}
 {id=Modulator_3, children [Random_2], properties {pitchMax=2, pitchMin=1.85}}
 {id=Modulator_6, children [Delay_6], properties {pitchMax=1, pitchMin=1}}
 {id=QualityLevel_0, children [Mixer_0, Mixer_2, WavePlayer_21, WavePlayer_21]}
 {id=Random_1, children [WavePlayer_17, WavePlayer_18], properties {weights [1,1]}}
 {id=Random_2, children [WavePlayer_12, WavePlayer_13], properties {weights [1,1]}}
 {id=Random_7, children [WavePlayer_5,WavePlayer_6,WavePlayer_7,WavePlayer_8], properties {weights [1,1,1,1]}}
 {id=Random_8, children [WavePlayer_0,WavePlayer_1,WavePlayer_2,WavePlayer_3,WavePlayer_4], properties {weights [1,1,1,1,1]}}
 {id=WavePlayer_0, properties {soundWaveAssetPtr=&/Metal_Cannon_Impact05}}
 {id=WavePlayer_1, properties {soundWaveAssetPtr=&/Metal_Cannon_Impact04}}
 {id=WavePlayer_12, properties {soundWaveAssetPtr=&Metal_Cannon_Impact05}}
 {id=WavePlayer_13, properties {soundWaveAssetPtr=&Metal_Cannon_Impact04}}
 {id=WavePlayer_16, properties {soundWaveAssetPtr=&/sfx_hitmarker_notification_critical_layer}}
 {id=WavePlayer_17, properties {soundWaveAssetPtr=&/player_impact_enemy_stereo_01}}
 {id=WavePlayer_18, properties {soundWaveAssetPtr=&/player_impact_enemy_stereo_02}}
 {id=WavePlayer_2, properties {soundWaveAssetPtr=&/Metal_Cannon_Impact03}}
 {id=WavePlayer_21, properties {soundWaveAssetPtr=&/HitMarker_Critical_opt}}
 {id=WavePlayer_3, properties {soundWaveAssetPtr=&/Metal_Cannon_Impact02}}
 {id=WavePlayer_4, properties {soundWaveAssetPtr=&/Metal_Cannon_Impact01}}
 {id=WavePlayer_5, properties {soundWaveAssetPtr=&player_impact_enemy_stereo_01}}
 {id=WavePlayer_6, properties {soundWaveAssetPtr=&player_impact_enemy_stereo_02}}
 {id=WavePlayer_7, properties {soundWaveAssetPtr=&/player_impact_enemy_stereo_03}}
 {id=WavePlayer_8, properties {soundWaveAssetPtr=&/player_impact_enemy_stereo_04}}
 {id=WavePlayer_9, properties {soundWaveAssetPtr=&sfx_hitmarker_notification_critical_layer}}
]
properties {
 attenuationSettings=/Game/Athena/Sounds/HitNotifications/Fort_HitMarker_Instigate_Att
 maxDistance=2097152
 soundClassObject=&Fort_2DHitNotifications_Instigate_PSM
 concurrencySet [&HitNotifications_TwoMaximum_StopOldest_VolumeScale]
}
