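# Sound cue for the hitmarker notification that signals a critical elimination, routed to the instigator's 2D hit notification sound class.
# A single elimination-crit wave passes through a default Modulator for variation, then a short fixed Delay before it plays.
# It uses the shared Fortnite instigator hitmarker attenuation and a critical-notification concurrency rule, named for stopping the oldest instance.
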SoundCue /FigmentCore/Sounds/HitNotifications/Hitmarker_Death_Critical_Cue
root=Delay_0
nodes [
 {id=Delay_0, children [Modulator_0], properties {delayMax=0.1, delayMin=0.1}}
 {id=Modulator_0, children [WavePlayer_0]}
 {id=WavePlayer_0, properties {soundWaveAssetPtr=&/HitNotification_New_ElimCrit_01}}
]
properties {
 attenuationSettings=/Game/Athena/Sounds/HitNotifications/Fort_HitMarker_Instigate_Att
 maxDistance=2097152
 soundClassObject=&Fort_2DHitNotifications_Instigate_PSM
 concurrencySet [/Game/Sounds/Concurrency_Rules/HitNotifications_Critical_StopOldest]
}
