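# Sound cue for the critical death hitmarker on the receiving side, a hit notification played for a fatal critical hit taken.
# A quality-level switch sends its first two slots to an equally weighted random pick of four variations and its last two to a single optimized wave.
# It uses shared hitmarker-receive attenuation, a 2D hit-notification sound class and a stop-oldest concurrency rule for critical notifications.
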SoundCue /FigmentCore/Sounds/HitNotifications/Hitmarker_Death_Critical_Receive_Cue
root=QualityLevel_1
nodes [
 {id=QualityLevel_1, children [Random_3,Random_3,WavePlayer_0,WavePlayer_0]}
 {id=Random_3, children [WavePlayer_1,WavePlayer_2,WavePlayer_3,WavePlayer_4], properties {weights [1,1,1,1]}}
 {id=WavePlayer_0, properties {soundWaveAssetPtr=&/HitMarker_Death_Critical_Receive_opt}}
 {id=WavePlayer_1, properties {soundWaveAssetPtr=&/HitMarker_Death_Critical_Receive_01}}
 {id=WavePlayer_2, properties {soundWaveAssetPtr=&/HitMarker_Death_Critical_Receive_02}}
 {id=WavePlayer_3, properties {soundWaveAssetPtr=&/HitMarker_Death_Critical_Receive_03}}
 {id=WavePlayer_4, properties {soundWaveAssetPtr=&/HitMarker_Death_Critical_Receive_04}}
]
properties {
 attenuationSettings=/Game/Athena/Sounds/HitNotifications/Fort_HitMarker_Receive_Att
 maxDistance=2097152
 soundClassObject=/Game/Sounds/FortSoundClassesAndModes/Fort_2DHitNotifications_Receive_PSM
 concurrencySet [/Game/Sounds/Concurrency_Rules/HitNotifications_Critical_StopOldest]
}
