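# Sound cue for the shield hitmarker notification, judging by its name and hit-notification sound class.
# A Random node picks one of four shield hit-notification waves with equal weight and passes it through a Delay node left at default settings.
# It uses a 2D instigator hit-notification sound class, hitmarker attenuation, and a concurrency set named for a two-voice, stop-oldest limit with volume scaling.
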SoundCue /FigmentCore/Sounds/HitNotifications/Hitmarker_Shield_Cue
root=Delay_0
nodes [
 {id=Delay_0, children [Random_4]}
 {id=Random_4, children [WavePlayer_0,WavePlayer_1,WavePlayer_2,WavePlayer_3], properties {weights [1,1,1,1]}}
 {id=WavePlayer_0, properties {soundWaveAssetPtr=&/HitNotification_New_Shield_01}}
 {id=WavePlayer_1, properties {soundWaveAssetPtr=&/HitNotification_New_Shield_02}}
 {id=WavePlayer_2, properties {soundWaveAssetPtr=&/HitNotification_New_Shield_03}}
 {id=WavePlayer_3, properties {soundWaveAssetPtr=&/HitNotification_New_Shield_04}}
]
properties {
 attenuationSettings=/Game/Athena/Sounds/HitNotifications/Att_HitMarker_Instigate_MS
 maxDistance=2097152
 soundClassObject=&Fort_2DHitNotifications_Instigate_PSM
 concurrencySet [&HitNotifications_TwoMaximum_StopOldest_VolumeScale]
}
