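# Sound cue for the hitmarker notification that plays when a shield is destroyed.
# A Delay node at default settings feeds a Random node that picks between two equally weighted shield-break waves.
# Playback uses instigator hitmarker attenuation, a 2D hit-notification sound class, and a critical concurrency rule named for stopping the oldest sound.
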SoundCue /FigmentCore/Sounds/HitNotifications/Hitmarker_Shield_Destroyed_Cue
root=Delay_0
nodes [
 {id=Delay_0, children [Random_0]}
 {id=Random_0, children [WavePlayer_0,WavePlayer_1], properties {weights [1,1]}}
 {id=WavePlayer_0, properties {soundWaveAssetPtr=&/HitNotification_New_ShieldBreak_01}}
 {id=WavePlayer_1, properties {soundWaveAssetPtr=&/HitNotification_New_ShieldBreak_02}}
]
properties {
 attenuationSettings=/Game/Athena/Sounds/HitNotifications/Fort_HitMarker_Instigate_Att
 maxDistance=2097152
 soundClassObject=&Fort_2DHitNotifications_Instigate_PSM
 concurrencySet [/Game/Sounds/Concurrency_Rules/HitNotifications_Critical_StopOldest]
}
