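# Sound cue for shield-impact hitmarker audio on received hits, routed to a 2D hit-notification sound class.
# A quality-level root selects a branch: one randomizes all ten shield impact waves, the others draw on smaller subsets of five or two.
# The smaller branches run through modulators for variation.
# Shared receive-hitmarker attenuation applies, and concurrency caps playback at two instances, stopping the oldest, judging by its name.
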SoundCue /FigmentCore/Sounds/HitNotifications/Hitmarker_Shield_Receive_Cue
root=QualityLevel_0
nodes [
 {id=Modulator_0, children [Random_2]}
 {id=Modulator_1, children [Random_3]}
 {id=Modulator_2, children [Random_0], properties {pitchMin=0.92}}
 {id=QualityLevel_0, children [Random_1,Modulator_2,Modulator_0,Modulator_1]}
 {id=Random_0, children [WavePlayer_14, WavePlayer_15, WavePlayer_16, WavePlayer_17, WavePlayer_18], properties {weights [1,1,1,1,1]}}
 {
  id=Random_1
  children [
   WavePlayer_13
   WavePlayer_4
   WavePlayer_5
   WavePlayer_6
   WavePlayer_7
   WavePlayer_8
   WavePlayer_9
   WavePlayer_10
   WavePlayer_11
   WavePlayer_12
  ]
  properties {weights [1,1,1,1,1,1,1,1,1,1]}
 }
 {id=Random_2, children [WavePlayer_19, WavePlayer_20], properties {weights [1,1]}}
 {id=Random_3, children [WavePlayer_21, WavePlayer_22], properties {weights [1,1]}}
 {id=WavePlayer_10, properties {soundWaveAssetPtr=&/cons_shield_impact_08}}
 {id=WavePlayer_11, properties {soundWaveAssetPtr=&/cons_shield_impact_09}}
 {id=WavePlayer_12, properties {soundWaveAssetPtr=&/cons_shield_impact_10}}
 {id=WavePlayer_13, properties {soundWaveAssetPtr=&/cons_shield_impact_01}}
 {id=WavePlayer_14, properties {soundWaveAssetPtr=&/cons_shield_impact_02}}
 {id=WavePlayer_15, properties {soundWaveAssetPtr=&/cons_shield_impact_04}}
 {id=WavePlayer_16, properties {soundWaveAssetPtr=&/cons_shield_impact_06}}
 {id=WavePlayer_17, properties {soundWaveAssetPtr=&cons_shield_impact_08}}
 {id=WavePlayer_18, properties {soundWaveAssetPtr=&cons_shield_impact_10}}
 {id=WavePlayer_19, properties {soundWaveAssetPtr=&cons_shield_impact_02}}
 {id=WavePlayer_20, properties {soundWaveAssetPtr=&cons_shield_impact_10}}
 {id=WavePlayer_21, properties {soundWaveAssetPtr=&cons_shield_impact_02}}
 {id=WavePlayer_22, properties {soundWaveAssetPtr=&cons_shield_impact_10}}
 {id=WavePlayer_4, properties {soundWaveAssetPtr=&cons_shield_impact_02}}
 {id=WavePlayer_5, properties {soundWaveAssetPtr=&/cons_shield_impact_03}}
 {id=WavePlayer_6, properties {soundWaveAssetPtr=&cons_shield_impact_04}}
 {id=WavePlayer_7, properties {soundWaveAssetPtr=&/cons_shield_impact_05}}
 {id=WavePlayer_8, properties {soundWaveAssetPtr=&cons_shield_impact_06}}
 {id=WavePlayer_9, properties {soundWaveAssetPtr=&/cons_shield_impact_07}}
]
properties {
 attenuationSettings=/Game/Athena/Sounds/HitNotifications/Fort_HitMarker_Receive_Att
 maxDistance=2097152
 soundClassObject=/Game/Sounds/FortSoundClassesAndModes/Fort_2DHitNotifications_Receive_Shield
 concurrencySet [&HitNotifications_TwoMaximum_StopOldest_VolumeScale]
}
